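# Physics asset for the Pink Emo Female pickaxe mesh, defining simple bodies for the handle and a two-bone dangling strap.
# Each of the three bones gets a small capsule, with the strap capsules offset along their bones.
# Constraints chain the handle to dyn_strap to dynStrap2, each joint allowing limited swing in both axes with twist limits set to zero.
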PhysicsAsset /BRCosmetics/Weapons/FORT_Melee/Pickaxe_Pink_Emo_Female/Meshes/Pink_Emo_Female_Axe_Physics
bodies {
 Melee_TwoHanded_Handle [{shapeType=Capsule, length=1, radius=1}]
 dynStrap2 [{shapeType=Capsule, length=1, radius=1, shapeName=dynStrap2_capsule, center=<5 0 0>}]
 dyn_strap [{shapeType=Capsule, length=1, radius=1, shapeName=dyn_strap_capsule, center=<5 0 0>}]
}
constraints [
 {bone1Name=dyn_strap, bone2Name=Melee_TwoHanded_Handle, swing1LimitDegrees=45, swing2LimitDegrees=45, twistLimitDegrees=0}
 {bone1Name=dynStrap2, bone2Name=dyn_strap, swing1LimitDegrees=45, swing2LimitDegrees=45, twistLimitDegrees=0}
]
